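# Foley sound cue for the eye blink in the Spyglass emote, played at reduced volume.
# A random node picks with equal weight among three eye-squish waves reused from the Demon pickaxe.
# It takes the emote foley sound class and property sheet, overrides attenuation with an occluded cone, and caps concurrency by stopping the oldest.
# A submix send also targets the music audio-analysis submix.
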SoundCue /BRCosmetics/Sounds/Emotes/Spyglass/Eyeblink/Emote_Spyglass_EyeBlink_Foley_Cue
root=Random_0
nodes [
 {id=Random_0, children [WavePlayer_7, WavePlayer_9, WavePlayer_10], properties {weights [1,1,1]}}
 {id=WavePlayer_10, properties {soundWaveAssetPtr=/BRCosmetics/Sounds/PickAxes/Demon/Demon_Pickaxe_Eye_Squish_09}}
 {id=WavePlayer_7, properties {soundWaveAssetPtr=/BRCosmetics/Sounds/PickAxes/Demon/Demon_Pickaxe_Eye_Squish_06}}
 {id=WavePlayer_9, properties {soundWaveAssetPtr=/BRCosmetics/Sounds/PickAxes/Demon/Demon_Pickaxe_Eye_Squish_08}}
]
properties {
 audioPropertiesSheet=/BRCosmetics/Sounds/Settings/PropertySheets/Emotes/APS_Cosmetic_Emote_Foley
 bOverrideAttenuation=true
 bOverrideConcurrency=true
 maxDistance=1400
 soundClassObject=/Game/Sounds/FortSoundClassesAndModes/Fort_Foley_Emote
 volumeMultiplier=0.2
 attenuationOverrides {
  attenuationShape=Cone
  bEnableOcclusion=true
  falloffDistance=1000
  occlusionInterpolationTime=0.5
  occlusionLowPassFilterFrequency=60000
  occlusionTraceChannel=ECC_GameTraceChannel11
  attenuationShapeExtents {y=90, z=15}
 }
 concurrencyOverrides {maxCount=3, resolutionRule=StopOldest}
 soundSubmixSends [{minSendLevel=0, soundSubmix=&Fortnite_AudioAnalysis_Music}]
}
